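# Idle Niagara effect for the Defect pickaxe, where Glow is the only enabled emitter.
# Glow's looping sprite and light particle follows a sampled skeletal mesh position each update, tinted by User.CO_ColorOverride.
# A waveform drives its Power, a curve modulates brightness, and a sine flickers the light, with time-of-day scalars.
# The disabled emitters hold a glow mesh, multicolour cube sprites using a local Quantizer module, and pixel-art shapes.
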NiagaraSystem /BRCosmetics/Weapons/FORT_Melee/Pickaxe_Defect/FX/NS_Pickaxe_Defect_Idle
dependencies {
 usedDataInterfaces [
  &NiagaraDataInterfaceCurve
  &NiagaraDataInterfaceArrayColor
  &NiagaraDataInterfaceVector2DCurve
  &NiagaraDataInterfaceArrayFloat
  &NiagaraDataInterfaceColorCurve
 ]
 usedDynamicInputs [&Vector2DFromCurve, &Vector2DFromFloat]
 usedModules [
  &self:SystemSpawnScript.SystemScriptSource.SystemScriptGraph.NiagaraNodeAssignment_0.SetVariables_#1
  &SystemState
 ]
}
emitters {
 Cubes {
  bEnabled=false
  emitterUpdateScript [
   &EmitterState(Life Cycle Mode=::NewEnumerator1, Loop Behavior=::NewEnumerator1, Loop Duration Mode=::NewEnumerator1, UseLoopDelay=true)
   &SpawnRate(SpawnRate=10, Use Spawn Probability=false)
   &SpawnBurst_Instantaneous(
    Spawn Count=&RandomRangeInt(Minimum=12, Maximum=14)
    Use Spawn Probability=false
    UseLoopCountLimit=false
   ) {enabled=false}
  ]
  particleSpawnScript [
   &InitializeParticle(
    Color=&SelectLinearColorFromArray(
     Array Sampling Mode=::NewEnumerator0
     Direct Array Index=?
     Color Selection Array=DataInterface() {propertyValues {ColorData [<rgba 3 1.1206 0.875 1>, <rgba 3 0.7969 2.6778 1>, <rgba 0.8125 3 1.2203 1>, <rgba 0.8281 1.8712 3 1>]}}
     Random Seed=0
    )
    Lifetime Max=3
    Lifetime Mode=::NewEnumerator1
    Sprite Rotation Mode=::NewEnumerator1
    Sprite Size Mode=::NewEnumerator3
    Uniform Sprite Size=&FloatFromCurve(
     FloatCurve=DataInterface() {propertyValues {Curve {keys [<at 0 1.5>,<at 0.2 1.5>,<at 0.8 4>,<at 1 4>]}}}
     CurveIndex=&RandomRangeFloat(Minimum=0, Maximum=1)
    )
    UsePositionOffset=false
    Write Material Random=true
   )
   &SkeletalMeshLocation(
    Apply Sampled Orientation to Mesh Particles=true
    Inherit Velocity=false
    "Inherit Velocity (Scale)"=0
    Interpolate Mesh Sampling=false
    MaskSpawnGroup=false
    Mesh Sampling Type=::NewEnumerator5
    Module.Override Randomness=false
    Sampled Position Offset=<0 0 10>
    Sampled Position Offset Coordinate Space:ENiagaraMeshTransforms=::NewEnumerator2
    Skeletal Mesh=$User.SkeletalMesh
   ) {module=SkeletalMeshLocation001}
   &InheritVelocity(Inherited Velocity Amount Scale=&VectorFromFloat(Value=5), Speed Limit=true, UseSpeedThreshold=false) {enabled=false}
   &ShapeLocation(MaskSpawnGroup=false, Override Seed=false, Ring Radius=2, Shape Primitive=::NewEnumerator4, UseOwnerScale=true)
   &AddVelocityFromPoint(Module.Use Falloff Distance=false, Velocity Strength=&RandomRangeFloat(Minimum=100, Maximum=200)) {enabled=false}
   &ApplyInitialForces()
   /SetVariables(
    Particles.SpriteAlignment=$Particles.Initial.ShapeLocation.ShapeNormal
    Particles.SpriteFacing=$Particles.Initial.ShapeLocation.ShapeBitangent
   ) {isSetParameters=true, module=SetVariables_#2}
   &InitialMeshOrientation(Mesh Orientation Mode=::NewEnumerator7, Rotation=<0 0 0.25>, Rotation Coordinate Space=::NewEnumerator2, Use Rotation Vector=true)
   &AlignSpriteToMeshOrientation(
    Mesh Orientation Relative Sprite Alignment Vector=$Particles.Initial.ShapeLocation.ShapeNormal
    Mesh Orientation Relative Sprite Facing Vector:Vector3f=$Particles.Initial.ShapeLocation.ShapeBitangent
    SetAlignment=true
    SetFacing=true
   )
   &CameraOffset(Camera Offset Amount=3) {enabled=false}
  ]
  particleUpdateScript [
   &Drag(Drag=&RandomRangeFloat(Minimum=10, Maximum=11), Use Linear Drag=true, Use Rotational Drag=false)
   &SolveForcesAndVelocity(Clamp Velocity=false, Limit Acceleration=false)
   &ParticleState()
   &ScaleSpriteSize(
    Uniform Curve Scale=0.8
    Uniform Curve Sprite Scale=DataInterface() {propertyValues {Curve {keys [
        {interpMode=RCIM_Cubic, tangentMode=RCTM_User, time=0, value=0}
        {interpMode=RCIM_Cubic, tangentMode=RCTM_User, time=0.1, value=1.625}
        {arriveTangent=-0.2045, interpMode=RCIM_Cubic, leaveTangent=-0.2045, tangentMode=RCTM_User, time=0.2, value=1}
        {arriveTangent=-1.983, interpMode=RCIM_Cubic, leaveTangent=-1.983, tangentMode=RCTM_User, time=1, value=0}
       ]}}}
   )
   &DynamicMaterialParameters(
    Index 0 Param 1=0.02
    Module.Param0WriteEnabled=true
    Module.Param1WriteEnabled=true
    Module.Param2WriteEnabled=true
    Module.Param3WriteEnabled=true
   )
   /SetVariables(Particles.Position=&Lerp_Vector(
     A=Linked() {linkedVariable {name=Particles.Initial.Position, type=NiagaraPosition}}
     B=&Add_Vector(
      A=Linked() {linkedVariable {name=Particles.Initial.Position, type=NiagaraPosition}}
      B=&TransformVector(Vector=<0 0 35>, Destination Space=::Simulation)
     )
     Alpha=$Particles.NormalizedAge
    )) {isSetParameters=true, module=SetVariables_#3}
   &self:Quantizer_04(SnapDistance=0.1) {module=Quantizer_04}
  ]
  rendererClasses [&NiagaraSpriteRendererProperties]
  renderers [{rendererClass=&NiagaraSpriteRendererProperties}]
 }
 Cubes001 {
  bEnabled=false
  emitterUpdateScript [
   &EmitterState(Life Cycle Mode=::NewEnumerator1, Loop Behavior=::NewEnumerator1, Loop Duration Mode=::NewEnumerator1, UseLoopDelay=true)
   &SpawnRate(SpawnRate=1.5, Use Spawn Probability=false)
   &SpawnBurst_Instantaneous(
    Spawn Count=&RandomRangeInt(Minimum=12, Maximum=14)
    Use Spawn Probability=false
    UseLoopCountLimit=false
   ) {enabled=false}
  ]
  particleSpawnScript [
   &InitializeParticle(
    Color=&SelectLinearColorFromArray(
     Array Sampling Mode=::NewEnumerator0
     Direct Array Index=?
     Color Selection Array=DataInterface() {propertyValues {ColorData [<rgba 3 1.1206 0.875 1>, <rgba 3 0.7969 2.6778 1>, <rgba 0.8125 3 1.2203 1>, <rgba 0.8281 1.8712 3 1>]}}
     Random Seed=0
    )
    Lifetime Max=0.6
    Lifetime Min=0.4
    Lifetime Mode=::NewEnumerator1
    Sprite Rotation Mode=::NewEnumerator1
    Sprite Size Mode=::NewEnumerator3
    Uniform Sprite Size=&FloatFromCurve(
     FloatCurve=DataInterface() {propertyValues {Curve {keys [<at 0 1.5>,<at 0.2 1.5>,<at 0.8 4>,<at 1 4>]}}}
     CurveIndex=&RandomRangeFloat(Minimum=0, Maximum=1)
    )
    UsePositionOffset=false
    Write Material Random=true
   )
   &InheritVelocity(Speed Limit=true, UseSpeedThreshold=false) {enabled=false, module=InheritVelocity001}
   &InheritVelocity(Inherited Velocity Amount Scale=&VectorFromFloat(Value=5), Speed Limit=true, UseSpeedThreshold=false) {enabled=false}
   &SkeletalMeshLocation(
    Apply Sampled Orientation to Mesh Particles=true
    Inherit Velocity=false
    "Inherit Velocity (Scale)"=0
    Interpolate Mesh Sampling=false
    MaskSpawnGroup=false
    Mesh Sampling Type=::NewEnumerator5
    Module.Override Randomness=false
    Sampled Position Offset=<0 0 30>
    Sampled Position Offset Coordinate Space:ENiagaraMeshTransforms=::NewEnumerator2
    Skeletal Mesh=$User.SkeletalMesh
   ) {module=SkeletalMeshLocation001}
   &ShapeLocation(
    Cylinder Height=40
    Cylinder Radius=7
    MaskSpawnGroup=false
    Override Seed=false
    Shape Primitive=::NewEnumerator1
    Surface Only=false
    UseOwnerScale=true
   )
   &AddVelocityFromPoint(
    Module.Use Falloff Distance=false
    Velocity Strength=&Multiply_Float(A=0.5, B=&RandomRangeFloat(Minimum=100, Maximum=400))
   ) {enabled=false}
   &ApplyInitialForces()
   /SetVariables(
    Particles.SpriteAlignment=$Particles.Initial.ShapeLocation.ShapeNormal
    Particles.SpriteFacing=$Particles.Initial.ShapeLocation.ShapeBitangent
   ) {enabled=false, isSetParameters=true, module=SetVariables_#2}
   &InitialMeshOrientation(Mesh Orientation Mode=::NewEnumerator7, Rotation=<0 0 0.25>, Rotation Coordinate Space=::NewEnumerator2, Use Rotation Vector=true)
   &AlignSpriteToMeshOrientation(SetAlignment=true, SetFacing=true) {enabled=false}
   &AlignSpriteToMeshOrientation(
    Mesh Orientation Relative Sprite Alignment Vector=$Particles.Initial.ShapeLocation.ShapeNormal
    Mesh Orientation Relative Sprite Facing Vector:Vector3f=$Particles.Initial.ShapeLocation.ShapeBitangent
    SetAlignment=true
    SetFacing=true
   ) {module=AlignSpriteToMeshOrientation001}
   &CameraOffset(Camera Offset Amount=3) {enabled=false}
  ]
  particleUpdateScript [
   &Drag(Drag=&RandomRangeFloat(Minimum=10, Maximum=11), Use Linear Drag=true, Use Rotational Drag=false)
   &SolveForcesAndVelocity(Clamp Velocity=false, Limit Acceleration=false)
   &ParticleState()
   &ScaleSpriteSize(
    Uniform Curve Scale=3
    Uniform Curve Sprite Scale=DataInterface() {propertyValues {Curve {keys [
        {interpMode=RCIM_Cubic, tangentMode=RCTM_User, time=0, value=0}
        {interpMode=RCIM_Cubic, tangentMode=RCTM_User, time=0.1, value=1.625}
        {interpMode=RCIM_Cubic, tangentMode=RCTM_User, time=0.2, value=1}
        {arriveTangent=-1.983, interpMode=RCIM_Cubic, leaveTangent=-1.983, tangentMode=RCTM_User, time=1, value=0}
       ]}}}
   )
   &DynamicMaterialParameters(
    Index 0 Param 1=0.02
    Module.Param0WriteEnabled=true
    Module.Param1WriteEnabled=true
    Module.Param2WriteEnabled=true
    Module.Param3WriteEnabled=true
   )
   /SetVariables(Particles.Position=&Lerp_Vector(
     A=Linked() {linkedVariable {name=Particles.Initial.Position, type=NiagaraPosition}}
     B=&Add_Vector(
      A=Linked() {linkedVariable {name=Particles.Initial.Position, type=NiagaraPosition}}
      B=&TransformVector(Vector=<0 0 35>, Destination Space=::Simulation)
     )
     Alpha=$Particles.NormalizedAge
    )) {enabled=false, isSetParameters=true, module=SetVariables_#3}
   &self:Quantizer_04(SnapDistance=0.1) {module=Quantizer_04}
  ]
  rendererClasses [&NiagaraSpriteRendererProperties]
  renderers [{rendererClass=&NiagaraSpriteRendererProperties}]
 }
 Cubes002 {
  bEnabled=false
  emitterUpdateScript [
   &EmitterState(Life Cycle Mode=::NewEnumerator1, Loop Behavior=::NewEnumerator1, Loop Duration Mode=::NewEnumerator1, UseLoopDelay=true)
   &SpawnRate(Spawn Probability=0.8, SpawnRate=4, Use Spawn Probability=true)
   &SpawnBurst_Instantaneous(
    Spawn Count=&RandomRangeInt(Minimum=12, Maximum=14)
    Use Spawn Probability=false
    UseLoopCountLimit=false
   ) {enabled=false}
  ]
  particleSpawnScript [
   &InitializeParticle(
    Color=&SelectLinearColorFromArray(
     Array Sampling Mode=::NewEnumerator0
     Direct Array Index=?
     Color Selection Array=DataInterface() {propertyValues {ColorData [<rgba 3 1.1206 0.875 1>, <rgba 3 0.7969 2.6778 1>, <rgba 0.8125 3 1.2203 1>, <rgba 0.8281 1.8712 3 1>]}}
     Random Seed=0
    )
    Lifetime Max=0.7
    Lifetime Min=0.4
    Lifetime Mode=::NewEnumerator1
    Sprite Rotation Mode=::NewEnumerator1
    Sprite Size Mode=::NewEnumerator1
    Uniform Sprite Size Max=5
    Uniform Sprite Size Min=2
    UsePositionOffset=false
    Write Material Random=true
   )
   &SkeletalMeshLocation(
    Apply Sampled Orientation to Mesh Particles=true
    Inherit Velocity=false
    "Inherit Velocity (Scale)"=0
    Interpolate Mesh Sampling=false
    MaskSpawnGroup=false
    Mesh Sampling Type=::NewEnumerator5
    Module.Override Randomness=false
    Sampled Position Offset=&RandomRangeVector(Minimum=<0 0 10>, Maximum=<0 0 40>)
    Sampled Position Offset Coordinate Space:ENiagaraMeshTransforms=::NewEnumerator2
    Skeletal Mesh=$User.SkeletalMesh
   ) {module=SkeletalMeshLocation001}
   &InheritVelocity(Inherited Velocity Amount Scale=&VectorFromFloat(Value=5), Speed Limit=true, UseSpeedThreshold=false) {enabled=false}
   &ShapeLocation(MaskSpawnGroup=false, Override Seed=false, Ring Radius=5, Shape Primitive=::NewEnumerator4, UseOwnerScale=true)
   &AddVelocityFromPoint(Module.Use Falloff Distance=false, Velocity Strength=&RandomRangeFloat(Minimum=100, Maximum=200)) {enabled=false}
   &ApplyInitialForces()
   /SetVariables(
    Particles.SpriteAlignment=$Particles.Initial.ShapeLocation.ShapeNormal
    Particles.SpriteFacing=$Particles.Initial.ShapeLocation.ShapeBitangent
   ) {isSetParameters=true, module=SetVariables_#2}
   &InitialMeshOrientation(Mesh Orientation Mode=::NewEnumerator7, Rotation=<0 0 0.25>, Rotation Coordinate Space=::NewEnumerator2, Use Rotation Vector=true)
   &AlignSpriteToMeshOrientation(
    Mesh Orientation Relative Sprite Alignment Vector=$Particles.Initial.ShapeLocation.ShapeNormal
    Mesh Orientation Relative Sprite Facing Vector:Vector3f=$Particles.Initial.ShapeLocation.ShapeBitangent
    SetAlignment=true
    SetFacing=true
   )
   &CameraOffset(Camera Offset Amount=3) {enabled=false}
  ]
  particleUpdateScript [
   &Drag(Drag=&RandomRangeFloat(Minimum=10, Maximum=11), Use Linear Drag=true, Use Rotational Drag=false)
   &SolveForcesAndVelocity(Clamp Velocity=false, Limit Acceleration=false)
   &ParticleState()
   &ScaleSpriteSize(
    Uniform Curve Scale=1.5
    Uniform Curve Sprite Scale=DataInterface() {propertyValues {Curve {keys [
        {interpMode=RCIM_Cubic, tangentMode=RCTM_User, time=0, value=2}
        {interpMode=RCIM_Cubic, tangentMode=RCTM_User, time=0.1, value=1.625}
        {arriveTangent=-0.2045, interpMode=RCIM_Cubic, leaveTangent=-0.2045, tangentMode=RCTM_User, time=0.2, value=1}
        {arriveTangent=-1.983, interpMode=RCIM_Cubic, leaveTangent=-1.983, tangentMode=RCTM_User, time=1, value=0}
       ]}}}
   )
   &DynamicMaterialParameters(
    Index 0 Param 1=0.02
    Module.Param0WriteEnabled=true
    Module.Param1WriteEnabled=true
    Module.Param2WriteEnabled=true
    Module.Param3WriteEnabled=true
   )
   /SetVariables(Particles.Position=&Lerp_Vector(
     A=Linked() {linkedVariable {name=Particles.Initial.Position, type=NiagaraPosition}}
     B=&Add_Vector(
      A=Linked() {linkedVariable {name=Particles.Initial.Position, type=NiagaraPosition}}
      B=&TransformVector(Vector=<0 0 35>, Destination Space=::Simulation)
     )
     Alpha=$Particles.NormalizedAge
    )) {enabled=false, isSetParameters=true, module=SetVariables_#3}
   &self:Quantizer_04(SnapDistance=0.1) {module=Quantizer_04}
  ]
  rendererClasses [&NiagaraSpriteRendererProperties]
  renderers [{rendererClass=&NiagaraSpriteRendererProperties}]
 }
 Glow {
  emitterSpawnScript [/SetVariables(Emitter.Depthfade=20, Emitter.Hardness=0.05, Emitter.Power=1) {isSetParameters=true}]
  emitterUpdateScript [
   &EmitterState(
    Inactive Response=::NewEnumerator2
    Life Cycle Mode=::NewEnumerator1
    Loop Behavior=::NewEnumerator1
    Loop Duration Mode=::NewEnumerator1
    UseLoopDelay=false
   )
   &SpawnBurst_Instantaneous(Use Spawn Probability=false, UseLoopCountLimit=false)
   /SetVariables(Emitter.Power:NiagaraFloat=&Add_Float(
     A=1.25
     B=&Waveform(
      Waveform Count=::NewEnumerator1
      "Waveform [1]"=::NewEnumerator0
      "[1] Phase Scale"=0.2215
      "[1] Frequency"=0.451
      "[1] Amplitude Min/Max"=<0 1>
      "[1] Scale"=0.95
      "[2] Average"=false
      "[2] Input Position"=0
      "[2] Phase"=$Emitter.Age
      "[2] Phase Scale"=0.234
      "[2] Compound Frequency"=2
      "[2] Frequency"=0.521
      "[2] Amplitude Min/Max"=<0 1>
      "[2] Scale"=1
      "[2] Bias"=0
      "[2] Exponent"=1
      "[2] Pulse Duration"=0.1
      "[2] Clip Triangle Min/Max"=<0 0>
      "[2] Random Seed"=0
      "[3] Phase"=$Emitter.Age
     )
    )) {isSetParameters=true}
  ]
  particleSpawnScript [
   &InitializeParticle(
    Color=&HueShiftLinearColor(
     Hue=0
     Saturation=1
     Value=1
     Linear Color=&MultiplyLinearColor(
      Linear Color=&HueShiftLinearColor(Hue=0, Saturation=1, Value=1.5, Linear Color=$User.CO_ColorOverride)
      Scale Factor=<1 1 1 0.05>
     )
    )
    Position Offset=<0 0 50>
    Sprite Size=<150 150>
    Sprite Size Mode=::NewEnumerator4
    UsePositionOffset=false
    Write Material Random=true
   )
   &CameraOffset()
  ]
  particleUpdateScript [
   &ParticleState(
    Kill Particles When Lifetime Has Elapsed=false
    Let Infinitely Lived Particles Die When Emitter Deactivates=true
    Loop Particles Lifetime=true
   )
   &SkeletalMeshLocation(
    Inherit Velocity=false
    "Inherit Velocity (Scale)"=0
    Interpolate Mesh Sampling=false
    MaskSpawnGroup=false
    Mesh Sampling Type=::NewEnumerator5
    Sampled Position Offset=<0 0 15>
    Sampled Position Offset Coordinate Space:ENiagaraMeshTransforms=::NewEnumerator2
    Skeletal Mesh=$System.Skeletal Mesh
    Socket Sampling Mode=::NewEnumerator5
   ) {module=SkeletalMeshLocation001}
   &ScaleColor(
    Scale RGB=&Multiply_VectorByFloat(
     Vector=<1 1 1>
     Float=&FloatFromCurve(
      FloatCurve=DataInterface() {propertyValues {Curve {keys [<at 0 0.8>,<at 0.1 0.9>,<at 0.3 0.7>,<at 0.4 0.8>,<at 0.5 0.8>,<at 0.6 1>,<at 0.8 0.75>,<at 0.9 0.95>,<at 1 0.8>]}}}
      Scale Curve=&Multiply_Float(A=&FNTimeOfDayScalar(Night Scale=0.5, Morning Scale=0.5, Day Scale=0.5, Evening Scale=0.5), B=1)
     )
    )
    ScaleA=true
    ScaleRGB=true
   )
   &FN_Light_Attributes(
    Color=&MultiplyLinearColor(
     Linear Color=&HueShiftLinearColor(Hue=0, Saturation=1, Value=10, Linear Color=$User.CO_ColorOverride)
     Scale Factor=&MakeVector4_FromFloat(Float=&Multiply_Float(
       A=&FNTimeOfDayScalar(Night Scale=0.2, Morning Scale=0.2, Day Scale=0.2, Evening Scale=0.2)
       B=&Sine(Normalized Angle=$Emitter.Age, Period=0.2, Scale=0.04, Bias=0.8)
      ))
    )
    Create Color Attribute=true
    Create Exponent Attribute=true
    Exponent=4
    Scale Color=0.5
   )
   /Niagara/Modules/Update/Color/ScaleColorBySpeed(Max Speed Threshold=500, Min Speed Threshold=0, RGB Scale Min=0.1, Source Velocity:Vector3f=$Engine.Owner.Velocity) {enabled=false}
  ]
  rendererClasses [&NiagaraSpriteRendererProperties, &NiagaraLightRendererProperties]
  renderers [
   {rendererClass=&NiagaraSpriteRendererProperties}
   {rendererClass=&NiagaraLightRendererProperties, rendererIndex=1}
  ]
 }
 GlowMesh {
  bEnabled=false
  emitterUpdateScript [
   &EmitterState(Life Cycle Mode=::NewEnumerator1, Loop Behavior=::NewEnumerator1, Loop Duration Mode=::NewEnumerator1, UseLoopDelay=false)
   &SpawnBurst_Instantaneous(Use Spawn Probability=false, UseLoopCountLimit=false)
  ]
  particleSpawnScript [&InitializeParticle(Color=<rgba 0 400 220.0003 1>, Mesh Scale Mode=::NewEnumerator3, UsePositionOffset=false, Write Material Random=true)]
  particleUpdateScript [&ParticleState(
    Kill Particles When Lifetime Has Elapsed=false
    Let Infinitely Lived Particles Die When Emitter Deactivates=true
    Loop Particles Lifetime=true
   )]
  rendererClasses [&NiagaraMeshRendererProperties]
  renderers [{rendererClass=&NiagaraMeshRendererProperties, rendererIndex=0}]
 }
 PixelArtShapes {
  bEnabled=false
  emitterUpdateScript [
   &EmitterState(
    Enable Distance Culling=true
    Max Distance Response=::KillImmediately
    Module.bUseMaxDistance=true
    Module.bUseMinDistance=false
    Scalability Mode=::NewEnumerator1
   )
   &SpawnRate(SpawnRate=5, Use Spawn Probability=false)
   &SpawnBurst_Instantaneous(Spawn Time=0.05, Use Spawn Probability=false, UseLoopCountLimit=false) {enabled=false}
  ]
  particleSpawnScript [
   &FN_SetSpawnGroup(Number of Spawn Groups=2)
   &InitializeParticle(
    Color=&SelectLinearColorFromArray(
     Array Sampling Mode=::NewEnumerator0
     Direct Array Index=?
     Color Selection Array=DataInterface() {propertyValues {ColorData [<rgba 3 1.1206 0.875 1>, <rgba 3 0.7969 2.6778 1>, <rgba 0.8125 3 1.2203 1>, <rgba 0.8281 1.8712 3 1>]}}
     Random Seed=0
    )
    Lifetime=&RandomRangeFloat(Minimum=0.3, Maximum=0.6)
    Sprite Rotation Angle=&SelectFloatFromArray(
     Array Sampling Mode=::NewEnumerator0
     Direct Array Index=?
     Float Selection Array=DataInterface() {propertyValues {FloatData [0,0.15,0,0]}}
     Random Seed=0
    )
    Sprite Rotation Mode=::NewEnumerator1
    Sprite Size Mode=::NewEnumerator1
    Uniform Sprite Size Max=60
    Uniform Sprite Size Min=20
    UsePositionOffset=false
    Write Material Random=true
   )
   &FN_PerParticleLOD(Enable SpriteSize Distance Scaling=true, Sprite Size Scale Far=<2.5 2.5>) {enabled=false}
   &CameraOffset(Camera Offset Amount=20) {enabled=false}
   &SkeletalMeshLocation(
    Inherit Velocity=false
    "Inherit Velocity (Scale)"=0
    Interpolate Mesh Sampling=false
    MaskSpawnGroup=false
    Mesh Sampling Type=::NewEnumerator5
    Module.Override Randomness=false
    Sampled Position Offset=&RandomRangeVector(Minimum=<0 0 -10>, Maximum=<0 0 -10>)
    Sampled Position Offset Coordinate Space:ENiagaraMeshTransforms=::NewEnumerator2
    Skeletal Mesh=$User.SkeletalMesh
   ) {module=SkeletalMeshLocation001}
   &ShapeLocation(MaskSpawnGroup=false, Override Seed=false, Sphere Radius=2.5, UseOwnerScale=true)
   /SetVariables(Particles.RandomFloat=&UniformRangedFloat(Minimum=0, Maximum=1)) {enabled=false, isSetParameters=true}
   /SetVariables(
    Particles.SubImageIndex=&RandomRangeFloat(Minimum=0, Maximum=3)
    Particles.SubImageIndexAlt:NiagaraFloat=&RandomRangeFloat(Minimum=0, Maximum=3)
   ) {enabled=false, isSetParameters=true}
   &InheritVelocity(Inherited Velocity Amount Scale=&VectorFromFloat(Value=0.5), Speed Limit=true, UseSpeedThreshold=false) {enabled=false}
   &AddVelocity(Velocity=&RandomRangeVector(Minimum=<0 0 -6>, Maximum=<0 0 -10>))
   &ApplyInitialForces()
  ]
  particleUpdateScript [
   &ParticleState()
   &InheritSourceMovement(
    Applied Force Scale=<-0.5 -0.5 0>
    Applied Position Scale=&VectorFromFloat(Value=&FloatFromCurve(
      FloatCurve=DataInterface() {propertyValues {Curve {keys [{interpMode=RCIM_Cubic, tangentMode=RCTM_User, time=0, value=1}, {interpMode=RCIM_Cubic, tangentMode=RCTM_User, time=0.8, value=0}]}}}
      Scale Curve=0.8
     ))
    ApplyForce=true
    ApplyForceFalloff=false
    ApplyPosition=true
    ApplyPositionFalloff=false
    Module.LimitForceSpeed=false
   )
   &SubUVAnimation(End Frame=4, Start Frame=1)
   &ScaleColor(
    Linear Color Curve=DataInterface() {propertyValues {
      AlphaCurve {keys [<at 0.1347 1>,<at 0.8992 0>]}
      BlueCurve {keys [<at 0.1978 1>,<at 0.9037 0.7659>]}
      GreenCurve {keys [<at 0.1978 0.125>,<at 0.9037 0.4597>]}
      RedCurve {keys [<at 0.1978 0.5625>,<at 0.9037 1>]}
     }}
    Scale Mode=::NewEnumerator2
   ) {enabled=false}
   &ScaleColor(
    Linear Color Curve=DataInterface() {propertyValues {
      AlphaCurve {keys [<at 0.3624 1>,<at 0.8992 0>]}
      BlueCurve {keys [<at 0.1978 1>]}
      GreenCurve {keys [<at 0.1978 1>]}
      RedCurve {keys [<at 0.1978 1>]}
     }}
    Scale Mode=::NewEnumerator2
   ) {enabled=false, module=ScaleColor002}
   &GravityForce(Gravity=<0 0 100>)
   &Drag(Drag=5, Use Linear Drag=true, Use Rotational Drag=true)
   &SolveForcesAndVelocity(Clamp Velocity=false, Limit Acceleration=false)
   &ScaleColor(
    Scale RGB=&MakeVectorFromLinearColorRGB(Color=&ColorFromCurve(ColorCurve=DataInterface() {propertyValues {
        AlphaCurve {keys [<at 0 1 RCIM_Cubic>,<at 0.5155 0 RCIM_Cubic>]}
        BlueCurve {keys [<at 0.1913 1 RCIM_Cubic>,<at 0.3986 0.6915 RCIM_Cubic>]}
        GreenCurve {keys [<at 0.1913 0.072 RCIM_Cubic>,<at 0.3986 0.5183 RCIM_Cubic>]}
        RedCurve {keys [<at 0.1913 0.6597 RCIM_Cubic>,<at 0.3986 1 RCIM_Cubic>]}
       }}))
    ScaleA=true
    ScaleRGB=true
   ) {enabled=false, module=ScaleColor001}
   /SetVariables(Particles.DistanceFromSpawn:NiagaraFloat=&Multiply_Float(
     A=&Multiply_Float(A=0.001, B=8)
     B=&VectorLength(Vector=&Subtract_Vector(
       A=Linked() {linkedVariable {name=Particles.Initial.Position, type=NiagaraPosition}}
       B=Linked() {linkedVariable {name=Particles.Position, type=NiagaraPosition}}
      ))
    )) {isSetParameters=true}
   &ScaleSpriteSize(
    Uniform Curve Scale=&Multiply_Float(
     A=&FloatFromCurve(FloatCurve=DataInterface() {propertyValues {Curve {keys [<at 0 1>,<at 0.5 0>]}}}, CurveIndex=$Particles.DistanceFromSpawn, Scale Curve=0.2)
     B=1
    )
    Uniform Curve Sprite Scale=DataInterface() {propertyValues {Curve {keys [{arriveTangent=2.6565, interpMode=RCIM_Constant, time=0, value=0.5}, <at 0.75 0.5 RCIM_Constant>]}}}
   ) {module=ScaleSpriteSize001}
   &ScaleColor(Scale RGB=&VectorFromFloat(Value=0.5), ScaleA=true, ScaleRGB=true) {module=ScaleColor003}
  ]
  rendererClasses [&NiagaraSpriteRendererProperties]
  renderers [{rendererClass=&NiagaraSpriteRendererProperties}]
 }
}
summary {
 systemName=NS_Pickaxe_Defect_Idle
 emitters [
  {emitterName=Glow, rendererClasses [&NiagaraSpriteRendererProperties, &NiagaraLightRendererProperties]}
  {bEnabled=false, emitterName=GlowMesh, rendererClasses [&NiagaraMeshRendererProperties]}
  {bEnabled=false, emitterName=Cubes, rendererClasses [&NiagaraSpriteRendererProperties]}
  {bEnabled=false, emitterName=Cubes001, rendererClasses [&NiagaraSpriteRendererProperties]}
  {bEnabled=false, emitterName=Cubes002, rendererClasses [&NiagaraSpriteRendererProperties]}
  {bEnabled=false, emitterName=PixelArtShapes, rendererClasses [&NiagaraSpriteRendererProperties]}
 ]
 userVariables [
  User.bIsFrontEndPreview {type=NiagaraBool, defaultValue=true}
  User.CO_ColorOverride {type=LinearColor, description="", defaultValue {struct=LinearColor, value=<rgba 1 0.2824 0 1>}}
  User.SkeletalMesh {
   type=NiagaraDataInterfaceSkeletalMesh
   defaultValue {
    struct=NiagaraExt_VariableValue_DataInterface
    value {dataInterface=&self:NiagaraDataInterfaceSkeletalMesh_0}
   }
  }
 ]
}
