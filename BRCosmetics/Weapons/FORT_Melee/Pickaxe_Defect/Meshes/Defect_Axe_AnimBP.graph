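# Animation Blueprint for the Defect pickaxe on the shared melee skeleton, copying its pose from another mesh and reorienting it.
# The AnimGraph applies an additive rotation to b_AR_Root in component space, then converts back to local space for output.
# Each update stores whether the cosmetic is displayed in a frontend context.
# That flag drives only a Blend Poses by bool node that is not connected to the output pose.
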AnimBlueprint /BRCosmetics/Weapons/FORT_Melee/Pickaxe_Defect/Meshes/Defect_Axe_AnimBP
target_skeleton=/Game/Weapons/FORT_Melee/Meshes/SK_Melee_Skeleton
graphs [
 AnimGraph {nodes [
   Blend Poses by bool(BlendPose_0=Copy Pose From Mesh(), bActiveValue=Get Is Front End() -> (Is Front End:bool)) {blendType=Linear}
   n1 = "Transform (Modify) Bone - Bone: b_AR_Root"(ComponentPose=Local To Component(LocalPose=Copy Pose From Mesh()), Alpha=1, Translation="0,0,0", Rotation="0,90,-90", Scale="1,1,1") {rotationMode=BMM_Additive, alphaBoolBlend {blendOption=Linear}, boneToModify {boneName=b_AR_Root}, rotation {roll=-90, yaw=90}}
   Output Pose(Result=Component To Local(ComponentPose=@n1)) {layerGroup=None}
  ]}
 EventGraph {nodes [
   Event BlueprintUpdateAnimation()
   > Set Is Front End(Is Front End:bool=IsCosmeticObjectDisplayedInFrontendContext!(CosmeticObject:Object Reference) -> (ReturnValue:bool)) -> (Output_Get:bool)
  ]}
]
variables [Is Front End {type=boolean}]
